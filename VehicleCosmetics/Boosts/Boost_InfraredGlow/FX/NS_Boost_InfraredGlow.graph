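# Niagara vehicle boost effect that trails glowing ribbons from filtered skeletal sockets, with a heat-map gradient as default.
# Three similar ribbon emitters launch particles backward at game-mode-masked speeds, inheriting vehicle movement that tapers over particle life.
# Two color painted variants with lifetime curves; NE_ColorsRibbon001 uses flat spawn colors, fading color scaling and larger Left/RightRibbon values.
# Each emitter has two ribbon renderers and sets Left/RightRibbon values by spawn group.
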NiagaraSystem /VehicleCosmetics/Boosts/Boost_InfraredGlow/FX/NS_Boost_InfraredGlow
dependencies {
 usedDataInterfaces [
  &NiagaraDataInterfaceColorCurve
  &NiagaraDataInterfaceArrayFloat
  &NiagaraDataInterfaceCurve
  &NiagaraDataInterfaceVectorCurve
  &NiagaraDataInterfaceSkeletalMesh
 ]
 usedDynamicInputs []
 usedModules [
  &self:SystemSpawnScript.SystemScriptSource.SystemScriptGraph.NiagaraNodeAssignment_0.SetVariables_#1
  &self:SystemSpawnScript.SystemScriptSource.SystemScriptGraph.NiagaraNodeAssignment_1.SetVariables_#2
  &SystemState
 ]
}
emitters {
 NE_ColorsRibbon {
  emitterUpdateScript [
   &EmitterState()
   &VC_FilteredSocketCount()
   &VC_RateQualityHandler(Epic Spawn Rate=30, High Spawn Rate=30, Low Spawn Rate=30, Use Custom Rate?=true)
   &VC_SpawnRibbon_v1(Number of Ribbons=$Emitter.PortCount, SpawnRate=$Emitter.Spawn Rate)
  ]
  particleSpawnScript [
   &VC_PaintedVariantsOverride()
   &VC_InitializeRibbon_v1(
    Color=$Particles.Color
    Custom Offset=true
    Enable Skeletal Position=true
    Lifetime=/VehicleCosmetics/FX/Niagara/Modules/VC_Mask_Floats(Coro Value:NiagaraFloat=1, FrontEnd Value:NiagaraFloat=1, FN Value:NiagaraFloat=1)
    Local Vector Offset=&AddVectorToPosition(Position=<5 0 0>, Vector=Linked() {linkedVariable {name=User.DTCarOffset, type=NiagaraPosition}})
    Number of Bones Per Boneflow group=2
    Offset Amount=1
    Ribbon Width=50
    Write Material Random=false
   )
   &AddVelocity(Velocity=&VC_Mask_Vectors(Coro Value:Vector3f=<-800 0 0>, FrontEnd Value:Vector3f=<-1100 0 0>, FN Value:Vector3f=<-1100 0 0>)) {module=AddVelocity001}
   /VehicleCosmetics/FX/VC_FXParents/Modules/DLR_ProductPreviews(GameModeID:NiagaraInt32=$System.GameModeID, bPreviewVehicleSpeed:NiagaraBool=$System.bPreviewVehicleSpeed) {enabled=false}
  ]
  particleUpdateScript [
   &ParticleState()
   &VC_PaintedVariantsOverride(
    Black=&ColorFromCurve(ColorCurve=DataInterface() {propertyValues {
       AlphaCurve {keys [<at 0 1 RCIM_Cubic>,<at 1 0 RCIM_Cubic>]}
       BlueCurve {keys [<at 0.063 0 RCIM_Cubic>,<at 0.254 0.0417 RCIM_Cubic>,<at 0.487 0>,<at 1 0>]}
       GreenCurve {keys [<at 0.063 0 RCIM_Cubic>,<at 0.254 0.0417 RCIM_Cubic>,<at 0.487 0>,<at 1 0>]}
       RedCurve {keys [<at 0.063 0 RCIM_Cubic>,<at 0.254 0.0417 RCIM_Cubic>,<at 0.487 0>,<at 1 0>]}
      }})
    BurntSienna=&ColorFromCurve(ColorCurve=DataInterface() {propertyValues {
       AlphaCurve {keys [<at 0 1 RCIM_Cubic>,<at 1 0 RCIM_Cubic>]}
       BlueCurve {keys [
         <at 0 0.0027 RCIM_Cubic>
         {arriveTangent=0.0165, interpMode=RCIM_Cubic, leaveTangent=0.0165, time=0.1891, value=0.0369}
         <at 0.4335 0.0099>
         <at 1 0>
        ]}
       GreenCurve {keys [
         <at 0 0.0126 RCIM_Cubic>
         {arriveTangent=0.0156, interpMode=RCIM_Cubic, leaveTangent=0.0156, time=0.1891, value=0.0723}
         <at 0.4335 0.0193>
         <at 1 0>
        ]}
       RedCurve {keys [
         <at 0 0.0469 RCIM_Cubic>
         {arriveTangent=0.012, interpMode=RCIM_Cubic, leaveTangent=0.012, time=0.1891, value=0.1946}
         <at 0.4335 0.0521>
         <at 1 0>
        ]}
      }})
    Cobalt=&ColorFromCurve(ColorCurve=DataInterface() {propertyValues {
       AlphaCurve {keys [<at 0 1 RCIM_Cubic>,<at 1 0 RCIM_Cubic>]}
       BlueCurve {keys [
         <at 0.063 0.9115 RCIM_Cubic>
         {arriveTangent=0.2089, interpMode=RCIM_Cubic, leaveTangent=0.2089, time=0.254, value=0.8957}
         <at 0.487 1>
         <at 1 0>
        ]}
       GreenCurve {keys [
         <at 0.063 0.0548 RCIM_Cubic>
         {arriveTangent=-0.0254, interpMode=RCIM_Cubic, leaveTangent=-0.0254, time=0.254, value=0.1681}
         <at 0.487 0.044>
         <at 1 0>
        ]}
       RedCurve {keys [
         <at 0.063 0.0089 RCIM_Cubic>
         {arriveTangent=-0.0051, interpMode=RCIM_Cubic, leaveTangent=-0.0051, time=0.254, value=0.0389}
         <at 0.487 0.0068>
         <at 1 0>
        ]}
      }})
    Crimson=&ColorFromCurve(ColorCurve=DataInterface() {propertyValues {
       AlphaCurve {keys [<at 0 1 RCIM_Cubic>,<at 1 0 RCIM_Cubic>]}
       BlueCurve {keys [
         <at 0.063 0 RCIM_Cubic>
         {arriveTangent=0.0213, interpMode=RCIM_Cubic, leaveTangent=0.0213, time=0.254, value=0.1042}
         <at 0.487 0.009>
         <at 1 0>
        ]}
       GreenCurve {keys [
         <at 0.063 0 RCIM_Cubic>
         {arriveTangent=0.0213, interpMode=RCIM_Cubic, leaveTangent=0.0213, time=0.254, value=0.1042}
         <at 0.487 0.009>
         <at 1 0>
        ]}
       RedCurve {keys [
         <at 0.063 0.8229 RCIM_Cubic>
         {arriveTangent=0.4177, interpMode=RCIM_Cubic, leaveTangent=0.4177, time=0.254, value=1}
         <at 0.487 1>
         <at 1 0>
        ]}
      }})
    DefaultColor=&ColorFromCurve(ColorCurve=DataInterface() {propertyValues {
       AlphaCurve {keys [<at 0 1 RCIM_Cubic>,<at 1 0 RCIM_Cubic>]}
       BlueCurve {keys [<at 0.0675 0 RCIM_Cubic>,<at 0.1761 0.0095>,<at 0.2551 0.0012>,<at 0.39 0.5>,<at 0.5184 0.5 RCIM_Cubic>]}
       GreenCurve {keys [<at 0.0675 0 RCIM_Cubic>,<at 0.1761 0.2168>,<at 0.2551 0.5>,<at 0.39 0.2196>,<at 0.5184 0.1007 RCIM_Cubic>]}
       RedCurve {keys [<at 0.0675 0.5 RCIM_Cubic>,<at 0.1761 0.5>,<at 0.2551 0.0951>,<at 0.39 0.0047>,<at 0.5184 0 RCIM_Cubic>]}
      }})
    ForestGreen=&ColorFromCurve(ColorCurve=DataInterface() {propertyValues {
       AlphaCurve {keys [<at 0 1 RCIM_Cubic>,<at 1 0 RCIM_Cubic>]}
       BlueCurve {keys [<at 0.063 0 RCIM_Cubic>,<at 0.254 0.0312 RCIM_Cubic>,<at 0.487 0>,<at 1 0>]}
       GreenCurve {keys [
         <at 0.063 0.3 RCIM_Cubic>
         {arriveTangent=0.1892, interpMode=RCIM_Cubic, leaveTangent=0.1892, time=0.254, value=0.5}
         <at 0.487 0.3802>
         <at 1 0>
        ]}
       RedCurve {keys [<at 0.063 0 RCIM_Cubic>,<at 0.254 0.1021 RCIM_Cubic>,<at 0.487 0>,<at 1 0>]}
      }})
    Grey=&ColorFromCurve(ColorCurve=DataInterface() {propertyValues {
       AlphaCurve {keys [<at 0 1 RCIM_Cubic>,<at 1 0 RCIM_Cubic>]}
       BlueCurve {keys [
         <at 0.063 0.1458 RCIM_Cubic>
         {arriveTangent=0.0491, interpMode=RCIM_Cubic, leaveTangent=0.0491, time=0.254, value=0.3333}
         <at 0.487 0.1667>
         <at 1 0>
        ]}
       GreenCurve {keys [
         <at 0.063 0.1458 RCIM_Cubic>
         {arriveTangent=0.0491, interpMode=RCIM_Cubic, leaveTangent=0.0491, time=0.254, value=0.3333}
         <at 0.487 0.1667>
         <at 1 0>
        ]}
       RedCurve {keys [
         <at 0.063 0.1458 RCIM_Cubic>
         {arriveTangent=0.0491, interpMode=RCIM_Cubic, leaveTangent=0.0491, time=0.254, value=0.3333}
         <at 0.487 0.1667>
         <at 1 0>
        ]}
      }})
    Lime=&ColorFromCurve(ColorCurve=DataInterface() {propertyValues {
       AlphaCurve {keys [<at 0 1 RCIM_Cubic>,<at 1 0 RCIM_Cubic>]}
       BlueCurve {keys [<at 0.063 0 RCIM_Cubic>,<at 0.254 0.0558 RCIM_Cubic>,<at 0.487 0>,<at 1 0>]}
       GreenCurve {keys [
         <at 0.063 0.313 RCIM_Cubic>
         {arriveTangent=0.2872, interpMode=RCIM_Cubic, leaveTangent=0.2872, time=0.254, value=0.6696}
         <at 0.487 0.4348>
         <at 1 0>
        ]}
       RedCurve {keys [
         <at 0.063 0.1615 RCIM_Cubic>
         {arriveTangent=0.0382, interpMode=RCIM_Cubic, leaveTangent=0.0382, time=0.254, value=0.4172}
         <at 0.487 0.1777>
         <at 1 0>
        ]}
      }})
    Orange=&ColorFromCurve(ColorCurve=DataInterface() {propertyValues {
       AlphaCurve {keys [<at 0 1 RCIM_Cubic>,<at 1 0 RCIM_Cubic>]}
       BlueCurve {keys [<at 0.063 0 RCIM_Cubic>,<at 0.254 0.0312 RCIM_Cubic>,<at 0.487 0>,<at 1 0>]}
       GreenCurve {keys [
         <at 0.063 0.1565 RCIM_Cubic>
         {arriveTangent=-0.1231, interpMode=RCIM_Cubic, leaveTangent=-0.1231, time=0.254, value=0.3354}
         <at 0.487 0.1043>
         <at 1 0>
        ]}
       RedCurve {keys [<at 0.063 1 RCIM_Cubic>,<at 0.254 1 RCIM_Cubic>,<at 0.487 1>,<at 1 0>]}
      }})
    Pink=&ColorFromCurve(ColorCurve=DataInterface() {propertyValues {
       AlphaCurve {keys [<at 0 1 RCIM_Cubic>,<at 1 0 RCIM_Cubic>]}
       BlueCurve {keys [
         <at 0.063 0.7282 RCIM_Cubic>
         {arriveTangent=0.5065, interpMode=RCIM_Cubic, leaveTangent=0.5065, time=0.254, value=0.9364}
         <at 0.487 0.9429>
         <at 1 0>
        ]}
       GreenCurve {keys [
         <at 0.063 0.0091 RCIM_Cubic>
         {arriveTangent=-0.0102, interpMode=RCIM_Cubic, leaveTangent=-0.0102, time=0.254, value=0.099}
         <at 0.487 0.0047>
         <at 1 0>
        ]}
       RedCurve {keys [
         <at 0.063 0.8229 RCIM_Cubic>
         {arriveTangent=0.4177, interpMode=RCIM_Cubic, leaveTangent=0.4177, time=0.254, value=1}
         <at 0.487 1>
         <at 1 0>
        ]}
      }})
    Purple=&ColorFromCurve(ColorCurve=DataInterface() {propertyValues {
       AlphaCurve {keys [<at 0 1 RCIM_Cubic>,<at 1 0 RCIM_Cubic>]}
       BlueCurve {keys [
         <at 0.063 0.8229 RCIM_Cubic>
         {arriveTangent=0.4177, interpMode=RCIM_Cubic, leaveTangent=0.4177, time=0.254, value=1}
         <at 0.487 1>
         <at 1 0>
        ]}
       GreenCurve {keys [
         <at 0.063 0.0109 RCIM_Cubic>
         {arriveTangent=-0.0109, interpMode=RCIM_Cubic, leaveTangent=-0.0109, time=0.254, value=0.0938}
         <at 0.487 0.0063>
         <at 1 0>
        ]}
       RedCurve {keys [
         <at 0.063 0.1774 RCIM_Cubic>
         {arriveTangent=-0.0372, interpMode=RCIM_Cubic, leaveTangent=-0.0372, time=0.254, value=0.3017}
         <at 0.487 0.1616>
         <at 1 0>
        ]}
      }})
    Saffron=&ColorFromCurve(ColorCurve=DataInterface() {propertyValues {
       AlphaCurve {keys [<at 0 1 RCIM_Cubic>,<at 1 0 RCIM_Cubic>]}
       BlueCurve {keys [
         <at 0.063 0.0333 RCIM_Cubic>
         {arriveTangent=-0.0765, interpMode=RCIM_Cubic, leaveTangent=-0.0765, time=0.254, value=0.1679}
         <at 0.487 0.0008>
         <at 1 0>
        ]}
       GreenCurve {keys [
         <at 0.063 0.3304 RCIM_Cubic>
         {arriveTangent=0.5795, interpMode=RCIM_Cubic, leaveTangent=0.5795, time=0.254, value=0.494}
         <at 0.487 0.576>
         <at 1 0>
        ]}
       RedCurve {keys [
         <at 0.063 0.3478 RCIM_Cubic>
         {arriveTangent=0.6564, interpMode=RCIM_Cubic, leaveTangent=0.6564, time=0.254, value=0.5217}
         <at 0.487 0.6261>
         <at 1 0>
        ]}
      }})
    SkyBlue=&ColorFromCurve(ColorCurve=DataInterface() {propertyValues {
       AlphaCurve {keys [<at 0 1 RCIM_Cubic>,<at 1 0 RCIM_Cubic>]}
       BlueCurve {keys [
         <at 0.063 0.8229 RCIM_Cubic>
         {arriveTangent=0.4177, interpMode=RCIM_Cubic, leaveTangent=0.4177, time=0.254, value=1}
         <at 0.487 1>
         <at 1 0>
        ]}
       GreenCurve {keys [
         <at 0.063 0.7421 RCIM_Cubic>
         {arriveTangent=0.4625, interpMode=RCIM_Cubic, leaveTangent=0.4625, time=0.254, value=0.9204}
         <at 0.487 0.9382>
         <at 1 0>
        ]}
       RedCurve {keys [<at 0.063 0 RCIM_Cubic>,<at 0.254 0.0208 RCIM_Cubic>,<at 0.487 0>,<at 1 0>]}
      }})
    TitaniumWhite=&ColorFromCurve(ColorCurve=DataInterface() {propertyValues {
       AlphaCurve {keys [<at 0 1 RCIM_Cubic>,<at 1 0 RCIM_Cubic>]}
       BlueCurve {keys [
         <at 0.063 0.3906 RCIM_Cubic>
         {arriveTangent=-0.2089, interpMode=RCIM_Cubic, leaveTangent=-0.2089, time=0.254, value=0.4531}
         <at 0.487 0.3021>
         <at 1 0>
        ]}
       GreenCurve {keys [
         <at 0.063 0.3906 RCIM_Cubic>
         {arriveTangent=-0.2089, interpMode=RCIM_Cubic, leaveTangent=-0.2089, time=0.254, value=0.4531}
         <at 0.487 0.3021>
         <at 1 0>
        ]}
       RedCurve {keys [
         <at 0.063 0.3906 RCIM_Cubic>
         {arriveTangent=-0.2089, interpMode=RCIM_Cubic, leaveTangent=-0.2089, time=0.254, value=0.4531}
         <at 0.487 0.3021>
         <at 1 0>
        ]}
      }})
   ) {module=VC_PaintedVariantsOverride001}
   &InheritSourceMovement(
    Applied Position Scale=&Multiply_VectorByFloat(Vector=<1 1 1>, Float=1)
    ApplyForce=false
    ApplyPosition=true
    ApplyPositionFalloff=false
   ) {enabled=false, module=InheritSourceMovement001}
   /SetVariables(
    Particles.LeftRibbon:NiagaraFloat=&SelectFloatFromArray(Direct Array Index=&SpawnGroup(), Float Selection Array=DataInterface() {propertyValues {FloatData [0,35]}})
    Particles.RightRibbon:NiagaraFloat=&SelectFloatFromArray(Direct Array Index=&SpawnGroup(), Float Selection Array=DataInterface() {propertyValues {FloatData [-35,0]}})
   ) {isSetParameters=true, module=SetVariables_#3}
   &InheritSourceMovement(
    Applied Position Scale=&VectorFromFloat(Value=&FloatFromCurve(FloatCurve=DataInterface() {propertyValues {Curve {
         defaultValue=1
         keys [
          {arriveTangent=-1.7885, interpMode=RCIM_Cubic, leaveTangent=-1.7885, tangentMode=RCTM_User, time=0.05, value=1}
          {interpMode=RCIM_Cubic, tangentMode=RCTM_User, time=1, value=0}
         ]
        }}}))
    ApplyForce=false
    ApplyPosition=true
    ApplyPositionFalloff=false
   )
   &SolveForcesAndVelocity(Clamp Velocity=false, Limit Acceleration=false)
   &DynamicMaterialParameters(
    Module.Index0Float0Write=true
    Module.Index0Float1Write=true
    Module.Index0Float2Write=true
    Module.Index0Float3Write=true
    Module.Index1Float0Write=true
    Module.Index1Float1Write=true
    Module.Index1Float2Write=true
    Module.Index1Float3Write=true
    Module.Index2Float0Write=true
    Module.Index2Float1Write=true
    Module.Index2Float2Write=true
    Module.Index2Float3Write=true
    Module.Index3Float0Write=true
    Module.Index3Float1Write=true
    Module.Index3Float2Write=true
    Module.Index3Float3Write=true
    Write Parameter Index 1=::NewEnumerator1
    Write Parameter Index 2=::NewEnumerator1
    Write Parameter Index 3=::NewEnumerator1
   ) {enabled=false}
   &VC_UpdateRibbon_v1(Enable Skeletal Position=true)
  ]
  rendererClasses [&NiagaraRibbonRendererProperties]
  renderers [
   {rendererClass=&NiagaraRibbonRendererProperties}
   {rendererClass=&NiagaraRibbonRendererProperties, rendererIndex=1}
  ]
 }
 NE_ColorsRibbon001 {
  emitterUpdateScript [
   &EmitterState()
   &VC_FilteredSocketCount()
   &VC_RateQualityHandler(Epic Spawn Rate=30, High Spawn Rate=30, Low Spawn Rate=30, Use Custom Rate?=true)
   &VC_SpawnRibbon_v1(Number of Ribbons=$Emitter.PortCount, SpawnRate=$Emitter.Spawn Rate)
  ]
  particleSpawnScript [
   &VC_PaintedVariantsOverride(
    Black=<rgba 0.0191 0.0191 0.0191 0.5>
    BurntSienna=<rgba 0.0642 0.0198 0.0079 0.7>
    Cobalt=<rgba 0 0.0185 1 0.7>
    Crimson=<rgba 0.4635 0 0.0126 0.6>
    DefaultAlpha=0.5
    DefaultColor=<rgba 0 0 1 0.5>
    ForestGreen=<rgba 0 0.5 0.0253 0.5>
    Grey=<rgba 0.2031 0.2031 0.2031 0.5>
    Lime=<rgba 0.2943 0.5885 0 0.5>
    Orange=<rgba 0.6724 0.2346 0 0.6>
    Pink=<rgba 0.5052 0.0605 0.3435 0.5>
    Purple=<rgba 0.2789 0 0.6795 0.5>
    Saffron=<rgba 0.9 0.69 0 0.5>
    SkyBlue=<rgba 0 0.357 0.5365 0.5>
    TitaniumWhite=<rgba 0.75 0.75 0.75 0.5>
   )
   &VC_InitializeRibbon_v1(
    Color=$Particles.Color
    Custom Offset=true
    Enable Skeletal Position=true
    Lifetime=&VC_Mask_Floats(Coro Value:NiagaraFloat=1, FrontEnd Value:NiagaraFloat=1, FN Value:NiagaraFloat=1)
    Local Vector Offset=&AddVectorToPosition(Position=<5 0 0>, Vector=Linked() {linkedVariable {name=User.DTCarOffset, type=NiagaraPosition}})
    Number of Bones Per Boneflow group=2
    Offset Amount=1
    Ribbon Width=50
    Write Material Random=false
   )
   &AddVelocity(Velocity=&VC_Mask_Vectors(Coro Value:Vector3f=<-800 0 0>, FrontEnd Value:Vector3f=<-1100 0 0>, FN Value:Vector3f=<-1100 0 0>)) {module=AddVelocity001}
   &DLR_ProductPreviews(GameModeID:NiagaraInt32=$System.GameModeID, bPreviewVehicleSpeed:NiagaraBool=$System.bPreviewVehicleSpeed) {enabled=false}
  ]
  particleUpdateScript [
   &ParticleState()
   &VC_PaintedVariantsOverride() {enabled=false, module=VC_PaintedVariantsOverride001}
   &InheritSourceMovement(
    Applied Position Scale=&Multiply_VectorByFloat(Vector=<1 1 1>, Float=1)
    ApplyForce=false
    ApplyPosition=true
    ApplyPositionFalloff=false
   ) {enabled=false, module=InheritSourceMovement001}
   /SetVariables(
    Particles.LeftRibbon:NiagaraFloat=&SelectFloatFromArray(Direct Array Index=&SpawnGroup(), Float Selection Array=DataInterface() {propertyValues {FloatData [0,100]}})
    Particles.RightRibbon:NiagaraFloat=&SelectFloatFromArray(Direct Array Index=&SpawnGroup(), Float Selection Array=DataInterface() {propertyValues {FloatData [-100,0]}})
   ) {isSetParameters=true, module=SetVariables_#3}
   &InheritSourceMovement(
    Applied Position Scale=&VectorFromFloat(Value=&FloatFromCurve(FloatCurve=DataInterface() {propertyValues {Curve {
         defaultValue=1
         keys [
          {arriveTangent=-1.7885, interpMode=RCIM_Cubic, leaveTangent=-1.7885, tangentMode=RCTM_User, time=0.05, value=1}
          {interpMode=RCIM_Cubic, tangentMode=RCTM_User, time=1, value=0}
         ]
        }}}))
    ApplyForce=false
    ApplyPosition=true
    ApplyPositionFalloff=false
   )
   &ScaleColor(
    Scale Alpha=&FloatFromCurve(
     FloatCurve=DataInterface() {propertyValues {Curve {defaultValue=1, keys [{tangentMode=RCTM_User, time=0, value=1}, {tangentMode=RCTM_User, time=1.012, value=-0.0421}]}}}
     Scale Curve=1.5
    )
    Scale RGB=&VectorFromCurve(
     VectorCurve=DataInterface() {propertyValues {XCurve {keys [<at 0 1>]}, YCurve {defaultValue=-0.0176, keys [<at 0.574 0.9967>]}, ZCurve {keys [<at 0.0611 0.9936>,<at 0.582 1.0005>]}}}
     Scale Curve=<0.5 0.5 0.5>
    )
    ScaleA=true
    ScaleRGB=true
   )
   &SolveForcesAndVelocity(Clamp Velocity=false, Limit Acceleration=false)
   &DynamicMaterialParameters(
    Module.Index0Float0Write=true
    Module.Index0Float1Write=true
    Module.Index0Float2Write=true
    Module.Index0Float3Write=true
    Module.Index1Float0Write=true
    Module.Index1Float1Write=true
    Module.Index1Float2Write=true
    Module.Index1Float3Write=true
    Module.Index2Float0Write=true
    Module.Index2Float1Write=true
    Module.Index2Float2Write=true
    Module.Index2Float3Write=true
    Module.Index3Float0Write=true
    Module.Index3Float1Write=true
    Module.Index3Float2Write=true
    Module.Index3Float3Write=true
    Write Parameter Index 1=::NewEnumerator1
    Write Parameter Index 2=::NewEnumerator1
    Write Parameter Index 3=::NewEnumerator1
   ) {enabled=false}
   &VC_UpdateRibbon_v1(Enable Skeletal Position=true)
  ]
  rendererClasses [&NiagaraRibbonRendererProperties]
  renderers [
   {rendererClass=&NiagaraRibbonRendererProperties}
   {rendererClass=&NiagaraRibbonRendererProperties, rendererIndex=1}
  ]
 }
 NE_ColorsRibbon002 {
  emitterUpdateScript [
   &EmitterState()
   &VC_FilteredSocketCount()
   &VC_RateQualityHandler(Epic Spawn Rate=30, High Spawn Rate=30, Low Spawn Rate=30, Use Custom Rate?=true)
   &VC_SpawnRibbon_v1(Number of Ribbons=$Emitter.PortCount, SpawnRate=$Emitter.Spawn Rate)
  ]
  particleSpawnScript [
   &VC_PaintedVariantsOverride()
   &VC_InitializeRibbon_v1(
    Color=$Particles.Color
    Custom Offset=true
    Enable Skeletal Position=true
    Lifetime=&VC_Mask_Floats(Coro Value:NiagaraFloat=1, FrontEnd Value:NiagaraFloat=1, FN Value:NiagaraFloat=1)
    Local Vector Offset=&AddVectorToPosition(Position=<5 0 0>, Vector=Linked() {linkedVariable {name=User.DTCarOffset, type=NiagaraPosition}})
    Number of Bones Per Boneflow group=2
    Offset Amount=1
    Ribbon Width=50
    Write Material Random=false
   )
   &AddVelocity(Velocity=&VC_Mask_Vectors(Coro Value:Vector3f=<-800 0 0>, FrontEnd Value:Vector3f=<-1100 0 0>, FN Value:Vector3f=<-1100 0 0>)) {module=AddVelocity001}
   &DLR_ProductPreviews(GameModeID:NiagaraInt32=$System.GameModeID, bPreviewVehicleSpeed:NiagaraBool=$System.bPreviewVehicleSpeed) {enabled=false}
  ]
  particleUpdateScript [
   &ParticleState()
   &VC_PaintedVariantsOverride(
    Black=&ColorFromCurve(ColorCurve=DataInterface() {propertyValues {
       AlphaCurve {keys [<at 0 1 RCIM_Cubic>,<at 1 0 RCIM_Cubic>]}
       BlueCurve {keys [
         <at 0.063 0 RCIM_Cubic>
         {arriveTangent=0.0246, interpMode=RCIM_Cubic, leaveTangent=0.0246, time=0.254, value=0.0469}
         <at 0.487 0.0104>
         <at 1 0>
        ]}
       GreenCurve {keys [
         <at 0.063 0 RCIM_Cubic>
         {arriveTangent=0.0246, interpMode=RCIM_Cubic, leaveTangent=0.0246, time=0.254, value=0.0469}
         <at 0.487 0.0104>
         <at 1 0>
        ]}
       RedCurve {keys [
         <at 0.063 0 RCIM_Cubic>
         {arriveTangent=0.0246, interpMode=RCIM_Cubic, leaveTangent=0.0246, time=0.254, value=0.0469}
         <at 0.487 0.0104>
         <at 1 0>
        ]}
      }})
    BurntSienna=&ColorFromCurve(ColorCurve=DataInterface() {propertyValues {
       AlphaCurve {keys [<at 0 1 RCIM_Cubic>,<at 1 0 RCIM_Cubic>]}
       BlueCurve {keys [
         <at 0 0.0027 RCIM_Cubic>
         {arriveTangent=0.0165, interpMode=RCIM_Cubic, leaveTangent=0.0165, time=0.1891, value=0.0369}
         <at 0.4335 0.0099>
         <at 1 0>
        ]}
       GreenCurve {keys [
         <at 0 0.0126 RCIM_Cubic>
         {arriveTangent=0.0156, interpMode=RCIM_Cubic, leaveTangent=0.0156, time=0.1891, value=0.0723}
         <at 0.4335 0.0193>
         <at 1 0>
        ]}
       RedCurve {keys [
         <at 0 0.0469 RCIM_Cubic>
         {arriveTangent=0.012, interpMode=RCIM_Cubic, leaveTangent=0.012, time=0.1891, value=0.1946}
         <at 0.4335 0.0521>
         <at 1 0>
        ]}
      }})
    Cobalt=&ColorFromCurve(ColorCurve=DataInterface() {propertyValues {
       AlphaCurve {keys [<at 0 1 RCIM_Cubic>,<at 1 0 RCIM_Cubic>]}
       BlueCurve {keys [
         <at 0.063 0.9115 RCIM_Cubic>
         {arriveTangent=0.2089, interpMode=RCIM_Cubic, leaveTangent=0.2089, time=0.254, value=0.8957}
         <at 0.487 1>
         <at 1 0>
        ]}
       GreenCurve {keys [
         <at 0.063 0.0548 RCIM_Cubic>
         {arriveTangent=0.0184, interpMode=RCIM_Cubic, leaveTangent=0.0184, time=0.254, value=0.135}
         <at 0.487 0.0626>
         <at 1 0>
        ]}
       RedCurve {keys [
         <at 0.063 0.0089 RCIM_Cubic>
         {arriveTangent=0.0404, interpMode=RCIM_Cubic, leaveTangent=0.0404, time=0.254, value=0}
         <at 0.487 0.026>
         <at 1 0>
        ]}
      }})
    Crimson=&ColorFromCurve(ColorCurve=DataInterface() {propertyValues {
       AlphaCurve {keys [<at 0 1 RCIM_Cubic>,<at 1 0 RCIM_Cubic>]}
       BlueCurve {keys [
         <at 0.063 0 RCIM_Cubic>
         {arriveTangent=0.0737, interpMode=RCIM_Cubic, leaveTangent=0.0737, time=0.254, value=0.1042}
         <at 0.487 0.0312>
         <at 1 0>
        ]}
       GreenCurve {keys [
         <at 0.063 0 RCIM_Cubic>
         {arriveTangent=0.0737, interpMode=RCIM_Cubic, leaveTangent=0.0737, time=0.254, value=0.1042}
         <at 0.487 0.0312>
         <at 1 0>
        ]}
       RedCurve {keys [
         <at 0.063 0.8229 RCIM_Cubic>
         {arriveTangent=0.4177, interpMode=RCIM_Cubic, leaveTangent=0.4177, time=0.254, value=1}
         <at 0.487 1>
         <at 1 0>
        ]}
      }})
    DefaultColor=&ColorFromCurve(ColorCurve=DataInterface() {propertyValues {
       AlphaCurve {keys [<at 0 1 RCIM_Cubic>,<at 1 0 RCIM_Cubic>]}
       BlueCurve {keys [<at 0.0675 0 RCIM_Cubic>,<at 0.1761 0.019>,<at 0.2551 0>,<at 0.39 1>,<at 0.5184 0.4348 RCIM_Cubic>]}
       GreenCurve {keys [<at 0.0675 0 RCIM_Cubic>,<at 0.1761 0.4335>,<at 0.2551 1>,<at 0.39 0.4338>,<at 0.5184 0.0876 RCIM_Cubic>]}
       RedCurve {keys [<at 0.0675 1 RCIM_Cubic>,<at 0.1761 1>,<at 0.2551 0.1882>,<at 0.39 0>,<at 0.5184 0 RCIM_Cubic>]}
      }})
    ForestGreen=&ColorFromCurve(ColorCurve=DataInterface() {propertyValues {
       AlphaCurve {keys [<at 0 1 RCIM_Cubic>,<at 1 0 RCIM_Cubic>]}
       BlueCurve {keys [
         <at 0.063 0 RCIM_Cubic>
         {arriveTangent=0.0234, interpMode=RCIM_Cubic, leaveTangent=0.0234, time=0.254, value=0.1901}
         <at 0.487 0.0099>
         <at 1 0>
        ]}
       GreenCurve {keys [
         <at 0.063 0.3 RCIM_Cubic>
         {arriveTangent=0.1892, interpMode=RCIM_Cubic, leaveTangent=0.1892, time=0.254, value=0.5}
         <at 0.487 0.3802>
         <at 1 0>
        ]}
       RedCurve {keys [
         <at 0.063 0 RCIM_Cubic>
         {arriveTangent=0.0234, interpMode=RCIM_Cubic, leaveTangent=0.0234, time=0.254, value=0.2369}
         <at 0.487 0.0099>
         <at 1 0>
        ]}
      }})
    Grey=&ColorFromCurve(ColorCurve=DataInterface() {propertyValues {
       AlphaCurve {keys [<at 0 1 RCIM_Cubic>,<at 1 0 RCIM_Cubic>]}
       BlueCurve {keys [
         <at 0.063 0.1458 RCIM_Cubic>
         {arriveTangent=0.0491, interpMode=RCIM_Cubic, leaveTangent=0.0491, time=0.254, value=0.3333}
         <at 0.487 0.1667>
         <at 1 0>
        ]}
       GreenCurve {keys [
         <at 0.063 0.1458 RCIM_Cubic>
         {arriveTangent=0.0491, interpMode=RCIM_Cubic, leaveTangent=0.0491, time=0.254, value=0.3333}
         <at 0.487 0.1667>
         <at 1 0>
        ]}
       RedCurve {keys [
         <at 0.063 0.1458 RCIM_Cubic>
         {arriveTangent=0.0491, interpMode=RCIM_Cubic, leaveTangent=0.0491, time=0.254, value=0.3333}
         <at 0.487 0.1667>
         <at 1 0>
        ]}
      }})
    Lime=&ColorFromCurve(ColorCurve=DataInterface() {propertyValues {
       AlphaCurve {keys [<at 0 1 RCIM_Cubic>,<at 1 0 RCIM_Cubic>]}
       BlueCurve {keys [<at 0.063 0 RCIM_Cubic>,<at 0.254 0.1652 RCIM_Cubic>,<at 0.487 0>,<at 1 0>]}
       GreenCurve {keys [
         <at 0.063 0.313 RCIM_Cubic>
         {arriveTangent=0.8096, interpMode=RCIM_Cubic, leaveTangent=0.8096, time=0.254, value=0.6696}
         <at 0.487 0.6562>
         <at 1 0>
        ]}
       RedCurve {keys [
         <at 0.063 0.1615 RCIM_Cubic>
         {arriveTangent=0.2517, interpMode=RCIM_Cubic, leaveTangent=0.2517, time=0.254, value=0.4622}
         <at 0.487 0.2682>
         <at 1 0>
        ]}
      }})
    Orange=&ColorFromCurve(ColorCurve=DataInterface() {propertyValues {
       AlphaCurve {keys [<at 0 1 RCIM_Cubic>,<at 1 0 RCIM_Cubic>]}
       BlueCurve {keys [
         <at 0.063 0 RCIM_Cubic>
         {arriveTangent=0.0491, interpMode=RCIM_Cubic, leaveTangent=0.0491, time=0.254, value=0.1019}
         <at 0.487 0.0208>
         <at 1 0>
        ]}
       GreenCurve {keys [
         <at 0.063 0.1565 RCIM_Cubic>
         {arriveTangent=-0.098, interpMode=RCIM_Cubic, leaveTangent=-0.098, time=0.254, value=0.3839}
         <at 0.487 0.115>
         <at 1 0>
        ]}
       RedCurve {keys [<at 0.063 1 RCIM_Cubic>,<at 0.254 1 RCIM_Cubic>,<at 0.487 1>,<at 1 0>]}
      }})
    Pink=&ColorFromCurve(ColorCurve=DataInterface() {propertyValues {
       AlphaCurve {keys [<at 0 1 RCIM_Cubic>,<at 1 0 RCIM_Cubic>]}
       BlueCurve {keys [
         <at 0.063 0.7282 RCIM_Cubic>
         {arriveTangent=0.5115, interpMode=RCIM_Cubic, leaveTangent=0.5115, time=0.254, value=0.9467}
         <at 0.487 0.945>
         <at 1 0>
        ]}
       GreenCurve {keys [
         <at 0.063 0.0091 RCIM_Cubic>
         {arriveTangent=0.0769, interpMode=RCIM_Cubic, leaveTangent=0.0769, time=0.254, value=0.2448}
         <at 0.487 0.0417>
         <at 1 0>
        ]}
       RedCurve {keys [
         <at 0.063 0.8229 RCIM_Cubic>
         {arriveTangent=0.4177, interpMode=RCIM_Cubic, leaveTangent=0.4177, time=0.254, value=1}
         <at 0.487 1>
         <at 1 0>
        ]}
      }})
    Purple=&ColorFromCurve(ColorCurve=DataInterface() {propertyValues {
       AlphaCurve {keys [<at 0 1 RCIM_Cubic>,<at 1 0 RCIM_Cubic>]}
       BlueCurve {keys [
         <at 0.063 0.8229 RCIM_Cubic>
         {arriveTangent=0.4177, interpMode=RCIM_Cubic, leaveTangent=0.4177, time=0.254, value=1}
         <at 0.487 1>
         <at 1 0>
        ]}
       GreenCurve {keys [
         <at 0.063 0.0109 RCIM_Cubic>
         {arriveTangent=-0.0258, interpMode=RCIM_Cubic, leaveTangent=-0.0258, time=0.254, value=0.1772}
         <at 0.487 0>
         <at 1 0>
        ]}
       RedCurve {keys [
         <at 0.063 0.1774 RCIM_Cubic>
         {arriveTangent=-0.0498, interpMode=RCIM_Cubic, leaveTangent=-0.0498, time=0.254, value=0.366}
         <at 0.487 0.1563>
         <at 1 0>
        ]}
      }})
    Saffron=&ColorFromCurve(ColorCurve=DataInterface() {propertyValues {
       AlphaCurve {keys [<at 0 1 RCIM_Cubic>,<at 1 0 RCIM_Cubic>]}
       BlueCurve {keys [
         <at 0.063 0.0333 RCIM_Cubic>
         {arriveTangent=-0.0765, interpMode=RCIM_Cubic, leaveTangent=-0.0765, time=0.254, value=0.1679}
         <at 0.487 0.0008>
         <at 1 0>
        ]}
       GreenCurve {keys [
         <at 0.063 0.3304 RCIM_Cubic>
         {arriveTangent=0.5795, interpMode=RCIM_Cubic, leaveTangent=0.5795, time=0.254, value=0.494}
         <at 0.487 0.576>
         <at 1 0>
        ]}
       RedCurve {keys [
         <at 0.063 0.3478 RCIM_Cubic>
         {arriveTangent=0.6564, interpMode=RCIM_Cubic, leaveTangent=0.6564, time=0.254, value=0.5217}
         <at 0.487 0.6261>
         <at 1 0>
        ]}
      }})
    SkyBlue=&ColorFromCurve(ColorCurve=DataInterface() {propertyValues {
       AlphaCurve {keys [<at 0 1 RCIM_Cubic>,<at 1 0 RCIM_Cubic>]}
       BlueCurve {keys [
         <at 0.063 0.8229 RCIM_Cubic>
         {arriveTangent=0.4177, interpMode=RCIM_Cubic, leaveTangent=0.4177, time=0.254, value=1}
         <at 0.487 1>
         <at 1 0>
        ]}
       GreenCurve {keys [
         <at 0.063 0.7421 RCIM_Cubic>
         {arriveTangent=0.4686, interpMode=RCIM_Cubic, leaveTangent=0.4686, time=0.254, value=0.9276}
         <at 0.487 0.9408>
         <at 1 0>
        ]}
       RedCurve {keys [
         <at 0.063 0 RCIM_Cubic>
         {arriveTangent=0.0983, interpMode=RCIM_Cubic, leaveTangent=0.0983, time=0.254, value=0.1094}
         <at 0.487 0.0417>
         <at 1 0>
        ]}
      }})
    TitaniumWhite=&ColorFromCurve(ColorCurve=DataInterface() {propertyValues {
       AlphaCurve {keys [<at 0 1 RCIM_Cubic>,<at 1 0 RCIM_Cubic>]}
       BlueCurve {keys [
         <at 0.063 0.3906 RCIM_Cubic>
         {arriveTangent=-0.2089, interpMode=RCIM_Cubic, leaveTangent=-0.2089, time=0.254, value=0.4531}
         <at 0.487 0.3021>
         <at 1 0>
        ]}
       GreenCurve {keys [
         <at 0.063 0.3906 RCIM_Cubic>
         {arriveTangent=-0.2089, interpMode=RCIM_Cubic, leaveTangent=-0.2089, time=0.254, value=0.4531}
         <at 0.487 0.3021>
         <at 1 0>
        ]}
       RedCurve {keys [
         <at 0.063 0.3906 RCIM_Cubic>
         {arriveTangent=-0.2089, interpMode=RCIM_Cubic, leaveTangent=-0.2089, time=0.254, value=0.4531}
         <at 0.487 0.3021>
         <at 1 0>
        ]}
      }})
   ) {module=VC_PaintedVariantsOverride001}
   &InheritSourceMovement(
    Applied Position Scale=&Multiply_VectorByFloat(Vector=<1 1 1>, Float=1)
    ApplyForce=false
    ApplyPosition=true
    ApplyPositionFalloff=false
   ) {enabled=false, module=InheritSourceMovement001}
   /SetVariables(
    Particles.LeftRibbon:NiagaraFloat=&SelectFloatFromArray(Direct Array Index=&SpawnGroup(), Float Selection Array=DataInterface() {propertyValues {FloatData [0,35]}})
    Particles.RightRibbon:NiagaraFloat=&SelectFloatFromArray(Direct Array Index=&SpawnGroup(), Float Selection Array=DataInterface() {propertyValues {FloatData [-35,0]}})
   ) {isSetParameters=true, module=SetVariables_#3}
   &InheritSourceMovement(
    Applied Position Scale=&VectorFromFloat(Value=&FloatFromCurve(FloatCurve=DataInterface() {propertyValues {Curve {
         defaultValue=1
         keys [
          {arriveTangent=-1.7885, interpMode=RCIM_Cubic, leaveTangent=-1.7885, tangentMode=RCTM_User, time=0.05, value=1}
          {interpMode=RCIM_Cubic, tangentMode=RCTM_User, time=1, value=0}
         ]
        }}}))
    ApplyForce=false
    ApplyPosition=true
    ApplyPositionFalloff=false
   )
   &SolveForcesAndVelocity(Clamp Velocity=false, Limit Acceleration=false)
   &DynamicMaterialParameters(
    Module.Index0Float0Write=true
    Module.Index0Float1Write=true
    Module.Index0Float2Write=true
    Module.Index0Float3Write=true
    Module.Index1Float0Write=true
    Module.Index1Float1Write=true
    Module.Index1Float2Write=true
    Module.Index1Float3Write=true
    Module.Index2Float0Write=true
    Module.Index2Float1Write=true
    Module.Index2Float2Write=true
    Module.Index2Float3Write=true
    Module.Index3Float0Write=true
    Module.Index3Float1Write=true
    Module.Index3Float2Write=true
    Module.Index3Float3Write=true
    Write Parameter Index 1=::NewEnumerator1
    Write Parameter Index 2=::NewEnumerator1
    Write Parameter Index 3=::NewEnumerator1
   ) {enabled=false}
   &VC_UpdateRibbon_v1(Enable Skeletal Position=true)
  ]
  rendererClasses [&NiagaraRibbonRendererProperties]
  renderers [
   {rendererClass=&NiagaraRibbonRendererProperties}
   {rendererClass=&NiagaraRibbonRendererProperties, rendererIndex=1}
  ]
 }
}
summary {
 systemName=NS_Boost_InfraredGlow
 emitters [
  {emitterName=NE_ColorsRibbon, rendererClasses [&NiagaraRibbonRendererProperties]}
  {emitterName=NE_ColorsRibbon001, rendererClasses [&NiagaraRibbonRendererProperties]}
  {emitterName=NE_ColorsRibbon002, rendererClasses [&NiagaraRibbonRendererProperties]}
 ]
 userVariables [
  User.bIsFrontendPreview {type=NiagaraBool, defaultValue=true}
  User.bIsPainted {type=NiagaraBool}
  User.bPreviewVehicleSpeed {type=NiagaraBool}
  User.ColorVariantID {type=NiagaraInt32}
  User.DTCarOffset {type=NiagaraPosition, description="", defaultValue {}}
  User.DTCarOffset {type=Vector3f, description="", defaultValue {struct=Vector3f, value=<0 0 0>}}
  User.GameModeID {type=NiagaraInt32, defaultValue=3}
  User.PaintedColor {type=LinearColor, description="", defaultValue {struct=LinearColor, value=<rgba 1 1 1 1>}}
  User.PaintedScalar {type=NiagaraFloat, defaultValue=1, description=""}
  User.Particles_MaterialRandom {type=NiagaraFloat, defaultValue=0, description=""}
  User.PortCount {type=NiagaraInt32, defaultValue=2}
  User.Power_AlphaPaintedR4 {type=NiagaraFloat, defaultValue=1, description=""}
  User.Power_EmissivePaintedR4 {type=NiagaraFloat, defaultValue=1, description=""}
 ]
}
